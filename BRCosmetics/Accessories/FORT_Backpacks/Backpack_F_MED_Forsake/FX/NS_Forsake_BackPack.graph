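# Niagara effect for the Forsake backpack cosmetic, adding flickering purple electricity and glows, or blue when the system's alt flag is set.
# Center_Electricity and Side_Electricity burst looping, partitioned ribbons bent along randomized Hermite splines between points sampled from the user skeletal mesh.
# AirGlows places pulsing lit sprites midway between sampled mesh points, with spawn count varying by platform quality.
# The alt-only sparkle emitter and both electricity light-attribute modules are disabled.
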NiagaraSystem /BRCosmetics/Accessories/FORT_Backpacks/Backpack_F_MED_Forsake/FX/NS_Forsake_BackPack
dependencies {
 usedDataInterfaces [
  &NiagaraDataInterfacePlatformSet
  &NiagaraDataInterfaceArrayInt32
  &NiagaraDataInterfaceCurve
  &NiagaraDataInterfaceArrayFloat2
  &NiagaraDataInterfaceArrayFloat3
  &NiagaraDataInterfaceArrayFloat
  &NiagaraDataInterfaceStaticMesh
  &NiagaraDataInterfaceArrayColor
 ]
 usedDynamicInputs []
 usedModules [
  &self:SystemSpawnScript.SystemScriptSource.SystemScriptGraph.NiagaraNodeAssignment_1.SetVariables_#1
  &SystemState
 ]
}
emitters {
 AirGlows {
  emitterUpdateScript [
   &EmitterState()
   &SpawnBurst_Instantaneous(
    Age:NiagaraFloat=$Emitter.Age
    Spawn Count=&MakeCustomIntFromBool(
     Bool=&IsPlatformSetActive(New Platform Set=DataInterface() {propertyValues {Platforms {qualityLevelMask=-29}}})
     True INT=1
     False INT=3
    )
    Use Spawn Probability=false
    UseLoopCountLimit=false
   )
  ]
  particleSpawnScript [
   &InitializeParticle(
    Color=&HueShiftLinearColor(
     Hue=0
     Saturation=1
     Value=1
     Linear Color=&MakeCustomLinearColorFromBool(Bool=$System.is alt bool, True Color=<rgba 0 0 1 0.4>, False Color=<rgba 0.8667 0 1 0.4>)
    )
    Sprite Size=&MakeCustomVector2DFromBool(
     Bool=&IsPlatformSetActive(New Platform Set=DataInterface() {propertyValues {Platforms {qualityLevelMask=-29}}})
     True Vector2D=<40 25>
     False Vector2D=&SelectVector2DFromArray(
      Direct Array Index=&ReturnParticleIDAttributes()
      Vector 2D Selection Array=DataInterface() {propertyValues {FloatData [<25 25>, <15 15>]}}
     )
    )
    Sprite Size Mode=::NewEnumerator4
    UsePositionOffset=false
    Write Material Random=true
   )
   &CameraOffset(Camera Offset Amount=15)
  ]
  particleUpdateScript [
   &ParticleState(
    Kill Particles When Lifetime Has Elapsed=false
    Let Infinitely Lived Particles Die When Emitter Deactivates=true
    Loop Particles Lifetime=true
   )
   &SkeletalMeshLocation(
    Bone / Socket Index=&ReturnParticleIDAttributes()
    Bone Sampling Mode=::NewEnumerator4
    MaskSpawnGroup=false
    Position Sampling=::NewEnumerator2
    Skeletal Mesh=$User.SkeletalMesh
    Use Sampled Mesh Attribute as Ribbon ID=false
   )
   &SkeletalMeshLocation(
    Bone / Socket Index=&Add_Integer(A=&ReturnParticleIDAttributes(), B=3)
    Bone Sampling Mode=::NewEnumerator4
    MaskSpawnGroup=false
    Position Sampling=::NewEnumerator2
    Skeletal Mesh=$User.SkeletalMesh
    Use Sampled Mesh Attribute as Ribbon ID=false
   ) {module=SkeletalMeshLocation001}
   /SetVariables(Particles.LerpPosition:NiagaraPosition=&ConvertVectorToPosition(
     Input Position=&Add_Vector(
      A=&Lerp_Vector(
       A=Linked() {linkedVariable {name=Particles.SkeletalMeshLocation.SampledPosition, type=NiagaraPosition}}
       B=Linked() {linkedVariable {name=Particles.SkeletalMeshLocation001.SampledPosition, type=NiagaraPosition}}
       Alpha=0.5
      )
      B=&TransformVector(
       Vector=&SelectVectorFromArray(
        Direct Array Index=&ReturnParticleIDAttributes()
        Vector Selection Array=DataInterface() {propertyValues {FloatData [<0 0 6>, <0 0 0>]}}
       )
       Destination Space=::Simulation
      )
     )
     Vector Is In Absolute Large World Space:NiagaraBool=false
    )) {isSetParameters=true}
   /SetVariables(Particles.Waveform:NiagaraFloat=&Waveform(
     Waveform Count=::NewEnumerator1
     "Waveform [1]"=::NewEnumerator10
     "[1] Phase"=&Add_Float(A=$Emitter.Age, B=&RandomRangeFloat(Minimum=0, Maximum=10))
     "[1] Frequency"=&RandomRangeFloat(Minimum=8, Maximum=12)
     "[1] Amplitude Min/Max"=<0.75 1>
     "[2] Average"=false
     "Waveform [2]"=::NewEnumerator10
     "[2] Input Position"=0
     "[2] Phase"=&Add_Float(A=$Emitter.Age, B=&RandomRangeFloat(Minimum=0, Maximum=10))
     "[2] Phase Scale"=1
     "[2] Compound Frequency"=2
     "[2] Frequency"=1
     "[2] Amplitude Min/Max"=<0 0.1>
     "[2] Scale"=1
     "[2] Bias"=0
     "[2] Exponent"=1
     "[2] Pulse Duration"=0.1
     "[2] Clip Triangle Min/Max"=<0 0>
     "[2] Random Seed"=0
    )) {isSetParameters=true}
   &FN_Light_Attributes(
    Color=&HueShiftLinearColor(Hue=0, Saturation=0.5, Value=1, Linear Color=$Particles.Color)
    Create Color Attribute=true
    Create Exponent Attribute=true
    Create Radius Attribute=true
    Enable Particle Light Enable=true
    Exponent=3
    Light Enabled=true
    Radius=&Multiply_Float(
     A=&SelectFloatFromArray(
      Direct Array Index=&ReturnParticleIDAttributes()
      Float Selection Array=DataInterface() {propertyValues {FloatData [22,13.7]}}
     )
     B=$Particles.Waveform
    )
    Scale Color=&Multiply_Float(A=$Particles.Waveform, B=8)
   )
  ]
  rendererClasses [&NiagaraSpriteRendererProperties, &NiagaraLightRendererProperties]
  renderers [
   {rendererClass=&NiagaraSpriteRendererProperties}
   {rendererClass=&NiagaraLightRendererProperties, rendererIndex=1}
  ]
 }
 AltSparkliesBetter {
  bEnabled=false
  emitterUpdateScript [
   &EmitterState()
   &SpawnRate(
    SpawnRate=&MakeCustomFloatFromBool(Bool=$System.is alt bool, True Float=2, False Float=0)
    Use Spawn Probability=false
   )
  ]
  particleSpawnScript [
   &InitializeParticle(
    Color=<rgba 21.7943 0 25 1>
    Lifetime Max=1.25
    Lifetime Min=0.75
    Lifetime Mode=::NewEnumerator1
    Mass Mode=::NewEnumerator3
    Sprite Rotation Angle=&RandomRangeFloat(Minimum=-10, Maximum=10)
    Sprite Rotation Mode=::NewEnumerator0
    Sprite Size Max=<3.3 6.6>
    Sprite Size Min=<3.3 6.6>
    Sprite Size Mode=::NewEnumerator2
    UsePositionOffset=false
    Write Material Random=true
   )
   /SetVariables(Particles.SectionSelect:NiagaraInt32=&RandomRangeInt(Minimum=0, Maximum=4)) {isSetParameters=true}
   &CameraOffset(Camera Offset Amount=2)
   /Niagara/Modules/Spawn/MeshInterface/SampleStaticMesh(
    OverrideRandomSeed=false
    StaticMesh=DataInterface() {propertyValues {
      DefaultMesh=/Game/Accessories/FORT_Backpacks/Backpack_F_MED_Forsake/FX/SM_forsakeSparkleSampler_All
      InstanceIndex=-1
      LODIndex=0
      PreviewMesh=&SM_forsakeSparkleSampler_All
      SoftSourceActor=null
      SourceMode=Default
      bAllowSamplingFromStreamingLODs=false
      bCaptureTransformsPerFrame=true
      bUsePhysicsBodyVelocity=false
      FilteredSockets []
      LODIndexUserParameter {parameter {name=None, typeDefHandle {registeredTypeIndex=108}, varData []}}
      MeshParameterBinding {parameter {name=None, typeDefHandle {registeredTypeIndex=1}, varData []}}
      SectionFilter {allowedMaterialSlots []}
     }}
   )
   /Niagara/Modules/Spawn/Location/StaticMeshLocation(Module.Enable Axis Constraint=false, Normal Offset=0.25)
   &CameraOffset(Camera Offset Amount=2) {module=CameraOffset001}
  ]
  particleUpdateScript [
   &ParticleState()
   &self:partition_Custom(New Linear Color:LinearColor=$Particles.SampleStaticMesh.MeshColor) {module=partition_Custom}
   &ScaleColor(
    Scale Alpha=&FloatFromCurve(FloatCurve=DataInterface() {propertyValues {Curve {keys [
         {arriveTangent=5.3218, interpMode=RCIM_Cubic, leaveTangent=5.3218, tangentMode=RCTM_User, time=0, value=0}
         {interpMode=RCIM_Cubic, leaveTangent=-4.4288, tangentMode=RCTM_Break, time=0.5, value=1}
         {interpMode=RCIM_Cubic, leaveTangent=-5.6888, time=1, value=0}
        ]}}})
    ScaleA=true
    ScaleRGB=true
   )
   &ScaleSpriteSize(
    Uniform Curve Scale=&RandomRangeFloat(Minimum=0.5, Maximum=0.8)
    Uniform Curve Sprite Scale=DataInterface() {propertyValues {Curve {keys [
        {arriveTangentWeight=0.332, interpMode=RCIM_Cubic, leaveTangentWeight=0.332, time=0, value=0.85}
        {arriveTangentWeight=0.1667, interpMode=RCIM_Cubic, leaveTangentWeight=0.1667, tangentMode=RCTM_Break, time=0.4, value=1}
        {
         arriveTangent=-3.1465
         arriveTangentWeight=0.1524
         interpMode=RCIM_Cubic
         leaveTangent=-3.1465
         leaveTangentWeight=0.1524
         tangentMode=RCTM_User
         time=1
         value=0
        }
       ]}}}
   )
   &SpriteRotationRate(Rotation Rate=&RandomRangeFloat(Minimum=10, Maximum=-10))
   &SkeletalMeshLocation(
    Bone / Socket Index=&SelectIntFromArray(Direct Array Index=$Output.partition_Custom.outputPartition, Int Selection Array=DataInterface() {propertyValues {IntData [0,0,4,5,1,2]}})
    Bone Sampling Mode=::NewEnumerator4
    MaskSpawnGroup=false
    Orientation Sampling=::NewEnumerator1
    Position Sampling=::NewEnumerator2
    Skeletal Mesh=$User.SkeletalMesh
    Use Sampled Mesh Attribute as Ribbon ID=false
   )
   &Color(Color=&SelectLinearColorFromArray(
     Array Sampling Mode=::NewEnumerator1
     Direct Array Index=$Output.partition_Custom.outputPartition
     Color Selection Array=DataInterface() {propertyValues {ColorData [<rgba 1 0 0 1>, <rgba 0 1 0 1>, <rgba 0 0 1 1>, <rgba 1 1 0 1>, <rgba 1 0 1 1>, <rgba 0 1 1 1>]}}
     Random Seed=?
    )) {enabled=false, module=Color001}
   /SetVariables(Particles.VelocityPosition:NiagaraPosition=&ConvertVectorToPosition(
     Input Position=&Add_Vector(
      A=&ConvertPositionToVector(Input Position=$Particles.Position, Convert To Absolute Large World Vector=false)
      B=Linked() {linkedVariable {name=Particles.SkeletalMeshLocation.SampledPosition, type=NiagaraPosition}}
     )
     Vector Is In Absolute Large World Space:NiagaraBool=false
    )) {isSetParameters=true}
  ]
  rendererClasses [&NiagaraSpriteRendererProperties]
  renderers [{rendererClass=&NiagaraSpriteRendererProperties}]
 }
 Center_Electricity {
  emitterSpawnScript [
   /SetVariables(Emitter.NumParticles:NiagaraInt32=&MakeCustomIntFromBool(
     Bool=&IsPlatformSetActive(New Platform Set=DataInterface() {propertyValues {Platforms {qualityLevelMask=-29}}})
     True INT=4
     False INT=6
    )) {isSetParameters=true, module=SetVariables_#2}
   /SetVariables(Emitter.NewScale:NiagaraFloat=&Divide_Float(
     A=1
     B=&OneMinusFloat(Float=&Divide_Float(A=1, B=&MakeFloatFromInt(Int=$Emitter.NumParticles)))
    )) {isSetParameters=true, module=SetVariables_#3}
  ]
  emitterUpdateScript [
   &EmitterState(
    Life Cycle Mode=::NewEnumerator1
    Loop Duration=&RandomRangeFloat(Minimum=0.3, Maximum=0.25)
    UseLoopDelay=false
   )
   &SpawnBurst_Instantaneous(Spawn Count=$Emitter.NumParticles, Use Spawn Probability=false, UseLoopCountLimit=false)
   &SpawnBurst_Instantaneous(
    Spawn Count=$Emitter.NumParticles
    Spawn Group=1
    Spawn Time=&RandomRangeFloat(Minimum=0.1, Maximum=0.3)
    Use Spawn Probability=false
    UseLoopCountLimit=false
   ) {module=SpawnBurst_Instantaneous001}
   /SetVariables(Emitter.RibbonLife:NiagaraFloat=&RandomRangeFloat(Minimum=0.35, Maximum=0.5, Evaluation Type=::NewEnumerator1)) {isSetParameters=true, module=SetVariables_#4}
   /SetVariables(Emitter.AlphaMulType:NiagaraBool=&SetBoolByFloatComparison(A=&RandomRangeFloat(Minimum=0, Maximum=1, Evaluation Type=::NewEnumerator1), B=0.25)) {isSetParameters=true, module=SetVariables_#5}
   /SetVariables(Emitter.RandomRibbonOffset:NiagaraFloat=&RandomRangeFloat(Minimum=0, Maximum=50, Evaluation Type=::NewEnumerator1)) {isSetParameters=true}
   /SetVariables() {enabled=false, isSetParameters=true, module=SetVariables_#6}
  ]
  particleSpawnScript [
   &InitializeRibbon(
    Color=&MakeCustomLinearColorFromBool(Bool=$System.is alt bool, True Color=<rgba 0 0 1 1>, False Color=<rgba 0.8667 0 1 1>)
    Lifetime=$Emitter.RibbonLife
    Ribbon Width=&RandomRangeFloat(Minimum=4, Maximum=4.5)
    UsePositionOffset=false
    Write Material Random=true
   )
   /SetVariables(Particles.Get AlphaMulType:NiagaraBool=$Emitter.AlphaMulType) {isSetParameters=true, module=SetVariables_#7}
   /SetVariables(Particles.RibbonU1Override:NiagaraFloat=&RandomRangeFloat(Minimum=0, Maximum=1)) {enabled=false, isSetParameters=true, module=SetVariables_#8}
   /SetVariables(Particles.GetRandomRibbonOffset:NiagaraFloat=$Emitter.RandomRibbonOffset) {isSetParameters=true}
   /SetVariables(Particles.get flip UV:NiagaraBool=$Emitter.flip UV) {enabled=false, isSetParameters=true, module=SetVariables_#9}
   /SetVariables(Particles.SpriteSize=<2 2>) {isSetParameters=true}
  ]
  particleUpdateScript [
   &ParticleState()
   &PartitionParticles(
    Particles Per Partition:NiagaraInt32=$Emitter.NumParticles
    Partition Count Mode:ENiagara_InfiniteLoopDuration=::NewEnumerator1
    Partition Mode:ENiagara_PartitionMode=::NewEnumerator1
    Write Partition Index to Ribbon Link Order:ENiagara_AttributeSamplingApplyOutput=::NewEnumerator0
    Write Partition to Ribbon ID=::NewEnumerator0
   )
   &SkeletalMeshLocation(
    Bone / Socket Index=&SelectIntFromArray(
     Direct Array Index=$Particles.PartitionParticles.Partition
     Direct Array Mode=::NewEnumerator1
     Int Selection Array=DataInterface() {propertyValues {IntData [0,2]}}
    )
    MaskSpawnGroup=false
    Mesh Sampling Type=::NewEnumerator5
    Skeletal Mesh=$User.SkeletalMesh
    Socket Sampling Mode=::NewEnumerator5
    Use Sampled Mesh Attribute as Ribbon ID=false
    Write To Particle Position=false
   )
   &SkeletalMeshLocation(
    Bone / Socket Index=&SelectIntFromArray(
     Direct Array Index=$Particles.PartitionParticles.Partition
     Direct Array Mode=::NewEnumerator1
     Int Selection Array=DataInterface() {propertyValues {IntData [1,3]}}
    )
    MaskSpawnGroup=false
    Mesh Sampling Type=::NewEnumerator5
    Skeletal Mesh=$User.SkeletalMesh
    Socket Sampling Mode=::NewEnumerator5
    Use Sampled Mesh Attribute as Ribbon ID=false
    Write To Particle Position=false
   ) {module=SkeletalMeshLocation001}
   /Game/Effects/Niagara/Modules/Curves/FN_HermiteSpline(
    Direction End:Vector3f=&Add_Vector(
     A=&Add_Vector(
      A=&Multiply_VectorByFloat(
       Vector=&ReturnMeshOrientationAxis(Mesh Orientation Axis=::XAxis, Mesh Orientation Quat=$Particles.SkeletalMeshLocation001.SampledRotation)
       Float=35
      )
      B=&Add_Vector(
       A=&RandomVector(Vector Scale=&FloatFromCurve(
         FloatCurve=DataInterface() {propertyValues {Curve {keys [
             {arriveTangentWeight=0.7975, leaveTangentWeight=0.7975, time=0.3, value=0}
             {arriveTangent=1.6439, arriveTangentWeight=0.6202, leaveTangent=1.6439, leaveTangentWeight=0.6202, time=0.75, value=1}
            ]}}}
         Scale Curve=&RandomRangeFloat(Minimum=0, Maximum=40)
        ))
       B=&RandomVector(Vector Scale=&RandomRangeFloat(Minimum=0, Maximum=10))
      )
     )
     B=&TransformVector(Vector=<0 -10 20>, Destination Space=::Simulation)
    )
    Direction Start:Vector3f=&Add_Vector(
     A=&Add_Vector(
      A=&Multiply_VectorByFloat(
       Vector=&ReturnMeshOrientationAxis(Mesh Orientation Axis=::XAxis, Mesh Orientation Quat=$Particles.SkeletalMeshLocation.SampledRotation)
       Float=30
      )
      B=&Add_Vector(
       A=&RandomVector(Vector Scale=&FloatFromCurve(
         FloatCurve=DataInterface() {propertyValues {Curve {keys [<at 0.3 0>, {arriveTangent=2.6952, leaveTangent=2.6952, time=0.75, value=1}]}}}
         Scale Curve=&RandomRangeFloat(Minimum=0, Maximum=40)
        ))
       B=&RandomVector(Vector Scale=&RandomRangeFloat(Minimum=0, Maximum=7))
      )
     )
     B=&TransformVector(Vector=<0 -25 0>, Destination Space=::Simulation)
    )
    Position Blend:NiagaraFloat=&Lerp_Float(
     A=-0.05
     B=1.05
     Alpha=&Multiply_Float(A=$Particles.PartitionParticles.Normalized Index In Partition, B=$Emitter.NewScale)
    )
    Position End:NiagaraPosition=$Particles.SkeletalMeshLocation001.SampledPosition
    Position Start:NiagaraPosition=$Particles.SkeletalMeshLocation.SampledPosition
    Write Particles Position=true
   )
   &ScaleRibbonWidth(Ribbon Width Scale=&Multiply_Float(
     A=&FloatFromCurve(FloatCurve=DataInterface() {propertyValues {Curve {keys [
          {arriveTangent=-2.4735, interpMode=RCIM_Cubic, leaveTangent=-2.4735, tangentMode=RCTM_User, time=0, value=1.5}
          {
           arriveTangentWeight=0.5275
           interpMode=RCIM_Cubic
           leaveTangentWeight=0.5124
           tangentMode=RCTM_Break
           tangentWeightMode=RCTWM_WeightedBoth
           time=0.5
           value=1
          }
          {arriveTangent=4.5078, interpMode=RCIM_Cubic, leaveTangent=4.5078, tangentMode=RCTM_User, time=1, value=1.5}
         ]}}})
     B=&FloatFromCurve(
      FloatCurve=DataInterface() {propertyValues {Curve {keys [
          {arriveTangent=4.4744, arriveTangentWeight=0.4802, leaveTangentWeight=0.4802, time=0, value=0.6}
          <at 0.1 1>
          {arriveTangent=0.0312, arriveTangentWeight=0.5275, leaveTangent=0.0312, leaveTangentWeight=0.5124, time=0.5, value=1}
          <at 0.9 1>
          {arriveTangent=-2.5305, leaveTangent=-2.5305, time=1, value=0.6}
         ]}}}
      CurveIndex=&Multiply_Float(A=$Particles.PartitionParticles.Normalized Index In Partition, B=$Emitter.NewScale)
      Scale Curve=1.5
     )
    ))
   &ScaleColor(
    Scale Alpha=&MakeCustomFloatFromBool(
     Bool=$Particles.Get AlphaMulType
     True Float=&FloatFromCurve(FloatCurve=DataInterface() {propertyValues {Curve {keys [
          {interpMode=RCIM_Cubic, leaveTangentWeight=0.1667, tangentMode=RCTM_Break, tangentWeightMode=RCTWM_WeightedBoth, time=0, value=0}
          {
           arriveTangentWeight=0.595
           interpMode=RCIM_Cubic
           leaveTangentWeight=0.607
           tangentMode=RCTM_Break
           tangentWeightMode=RCTWM_WeightedBoth
           time=0.5
           value=1
          }
         ]}}})
     False Float=&Waveform("Waveform [1]"=::NewEnumerator0, "[1] Frequency"=8, "[1] Amplitude Min/Max"=<0 1>)
    )
    ScaleA=true
    ScaleRGB=true
   )
   /SetVariables(Particles.RibbonU1Override:NiagaraFloat=&Multiply_Float(
     A=&Add_Float(
      A=&Multiply_Float(A=$Particles.PartitionParticles.Normalized Index In Partition, B=$Emitter.NewScale)
      B=$Particles.GetRandomRibbonOffset
     )
     B=1
    )) {isSetParameters=true}
   &DynamicMaterialParameters(
    Index 0 Param 1=1
    Index 0 Param 2=&FloatFromCurve(FloatCurve=DataInterface() {propertyValues {Curve {keys [<at 0.75 5>,<at 1 -1>]}}})
    Index 0 Param 3=1
    Index 0 Param 4=&RandomRangeFloat(Minimum=-0.1, Maximum=0.1)
    Module.Param0WriteEnabled=true
    Module.Param1WriteEnabled=true
    Module.Param2WriteEnabled=true
    Module.Param3WriteEnabled=true
   )
   &FN_Light_Attributes(
    Color=&HueShiftLinearColor(Hue=0, Saturation=0.5, Value=1, Linear Color=$Particles.Color)
    Create Color Attribute=true
    Create Exponent Attribute=true
    Create Radius Attribute=true
    Enable Particle Light Enable=true
    Exponent=6
    Light Enabled:NiagaraBool=&Boolean_AndOperation(
     Bool1=&SetBoolByFloatComparison(A=&RandomRangeFloat(Minimum=0, Maximum=1), B=0.9)
     Bool2=&SetBoolByFloatComparison(
      A=&FloatFromCurve(
       FloatCurve=DataInterface() {propertyValues {Curve {keys [<at 0 1>,<at 0.05 0>,<at 0.3271 0>,<at 0.5 1>,<at 0.6623 0>,<at 0.95 0>,<at 1 1>]}}}
       CurveIndex=&Multiply_Float(A=$Particles.PartitionParticles.Normalized Index In Partition, B=$Emitter.NewScale)
      )
      B=0.5
     )
    )
    Radius=25
    Scale Color=&Multiply_Float(
     A=&Waveform(
      "Waveform [1]"=::NewEnumerator0
      "[1] Frequency"=&RandomRangeFloat(Minimum=4, Maximum=5)
      "[1] Amplitude Min/Max"=<0.75 1>
      "[1] Bias"=1
     )
     B=1.5
    )
   ) {enabled=false}
  ]
  rendererClasses [
   &NiagaraRibbonRendererProperties
   &NiagaraLightRendererProperties
   &NiagaraSpriteRendererProperties
  ]
  renderers [
   {rendererClass=&NiagaraRibbonRendererProperties}
   {rendererClass=&NiagaraLightRendererProperties, rendererIndex=1}
   {rendererClass=&NiagaraSpriteRendererProperties, rendererIndex=2}
  ]
 }
 Side_Electricity {
  emitterSpawnScript [
   /SetVariables(Emitter.NumParticles:NiagaraInt32=&RandomRangeInt(Minimum=4, Maximum=6)) {isSetParameters=true, module=SetVariables_#2}
   /SetVariables(Emitter.NewScale:NiagaraFloat=&Divide_Float(
     A=1
     B=&OneMinusFloat(Float=&Divide_Float(A=1, B=&MakeFloatFromInt(Int=$Emitter.NumParticles)))
    )) {isSetParameters=true, module=SetVariables_#3}
  ]
  emitterUpdateScript [
   &EmitterState(
    Life Cycle Mode=::NewEnumerator1
    Loop Duration=&RandomRangeFloat(Minimum=0.15, Maximum=0.1)
    UseLoopDelay=false
   )
   &SpawnBurst_Instantaneous(Spawn Count=$Emitter.NumParticles, Use Spawn Probability=false, UseLoopCountLimit=false)
   &SpawnBurst_Instantaneous(Spawn Count=$Emitter.NumParticles, Use Spawn Probability=false, UseLoopCountLimit=false) {module=SpawnBurst_Instantaneous001}
   /SetVariables(Emitter.Left Or Right:NiagaraBool=&RandomBool(Evaluation Type=::NewEnumerator1)) {isSetParameters=true}
   /SetVariables(Emitter.RibbonLife:NiagaraFloat=&RandomRangeFloat(Minimum=0.1, Maximum=0.5, Evaluation Type=::NewEnumerator1)) {isSetParameters=true, module=SetVariables_#4}
   /SetVariables(Emitter.AlphaMulType:NiagaraBool=&SetBoolByFloatComparison(A=&RandomRangeFloat(Minimum=0, Maximum=1, Evaluation Type=::NewEnumerator1), B=0.5)) {isSetParameters=true, module=SetVariables_#5}
   /SetVariables(Emitter.flip UV:NiagaraBool=&RandomBool(Evaluation Type=::NewEnumerator1)) {isSetParameters=true, module=SetVariables_#6}
  ]
  particleSpawnScript [
   &InitializeRibbon(
    Color=&MakeCustomLinearColorFromBool(Bool=$System.is alt bool, True Color=<rgba 0 0 1 1>, False Color=<rgba 0.8667 0 1 1>)
    Lifetime=$Emitter.RibbonLife
    Ribbon Width=&RandomRangeFloat(Minimum=10, Maximum=8.75)
    UsePositionOffset=false
    Write Material Random=true
   )
   /SetVariables(Particles.GetLeftOrRight:NiagaraBool=$Emitter.Left Or Right) {isSetParameters=true}
   /SetVariables(Particles.Get AlphaMulType:NiagaraBool=$Emitter.AlphaMulType) {isSetParameters=true, module=SetVariables_#7}
   /SetVariables(Particles.get flip UV:NiagaraBool=$Emitter.flip UV) {isSetParameters=true, module=SetVariables_#9}
   /SetVariables(Particles.SpriteSize=<2 2>) {isSetParameters=true}
   /SetVariables(Particles.SpriteCameraOffset=1.5) {isSetParameters=true}
  ]
  particleUpdateScript [
   &ParticleState()
   /SetVariables(Particles.GetRandomRibbonOffset:NiagaraFloat=&self:ReNormalize()) {isSetParameters=true}
   &PartitionParticles(
    Particles Per Partition:NiagaraInt32=$Emitter.NumParticles
    Partition Count Mode:ENiagara_InfiniteLoopDuration=::NewEnumerator1
    Partition Mode:ENiagara_PartitionMode=::NewEnumerator1
    Write Partition Index to Ribbon Link Order:ENiagara_AttributeSamplingApplyOutput=::NewEnumerator0
    Write Partition to Ribbon ID=::NewEnumerator0
   )
   &SkeletalMeshLocation(
    Bone / Socket Index=&SelectIntFromArray(
     Direct Array Index=$Particles.PartitionParticles.Partition
     Direct Array Mode=::NewEnumerator1
     Int Selection Array=DataInterface() {propertyValues {IntData [4,7]}}
    )
    MaskSpawnGroup=false
    Mesh Sampling Type=::NewEnumerator5
    Position Sampling=::NewEnumerator2
    Sampled Position Offset=&Add_Vector(
     A=&Multiply_Vector(A=&RandomVector(Vector Scale=1.25), B=<0 1 1>)
     B=<0.8 0 0>
    )
    Skeletal Mesh=$User.SkeletalMesh
    Socket Sampling Mode=::NewEnumerator5
    Use Sampled Mesh Attribute as Ribbon ID=false
   )
   &SkeletalMeshLocation(
    Bone / Socket Index=&SelectIntFromArray(
     Direct Array Index=$Particles.PartitionParticles.Partition
     Direct Array Mode=::NewEnumerator1
     Int Selection Array=DataInterface() {propertyValues {IntData [5,8]}}
    )
    MaskSpawnGroup=false
    Mesh Sampling Type=::NewEnumerator5
    Position Sampling=::NewEnumerator2
    Skeletal Mesh=$User.SkeletalMesh
    Socket Sampling Mode=::NewEnumerator5
    Use Sampled Mesh Attribute as Ribbon ID=false
   ) {module=SkeletalMeshLocation001}
   &SkeletalMeshLocation(
    Bone / Socket Index=&SelectIntFromArray(
     Direct Array Index=$Particles.PartitionParticles.Partition
     Direct Array Mode=::NewEnumerator1
     Int Selection Array=DataInterface() {propertyValues {IntData [6,9]}}
    )
    MaskSpawnGroup=false
    Mesh Sampling Type=::NewEnumerator5
    Position Sampling=::NewEnumerator2
    Skeletal Mesh=$User.SkeletalMesh
    Socket Sampling Mode=::NewEnumerator5
    Use Sampled Mesh Attribute as Ribbon ID=false
   ) {module=SkeletalMeshLocation002}
   &FN_HermiteSpline(
    Direction End:Vector3f=&Add_Vector(
     A=&Multiply_VectorByFloat(
      Vector=&ReturnMeshOrientationAxis(Mesh Orientation Axis=::XAxis, Mesh Orientation Quat=$Particles.SkeletalMeshLocation001.SampledRotation)
      Float=15
     )
     B=&RandomVector(
      Vector Scale=&RandomRangeFloat(Minimum=0, Maximum=3)
      Constrain to Hemisphere=true
      Hemisphere Constraint Vector=&ReturnMeshOrientationAxis(Mesh Orientation Axis=::XAxis, Mesh Orientation Quat=$Particles.SkeletalMeshLocation001.SampledRotation)
     )
    )
    Direction Start:Vector3f=&Add_Vector(
     A=&Multiply_VectorByFloat(
      Vector=&ReturnMeshOrientationAxis(Mesh Orientation Axis=::XAxis, Mesh Orientation Quat=$Particles.SkeletalMeshLocation.SampledRotation)
      Float=8
     )
     B=&RandomVector(
      Vector Scale=&RandomRangeFloat(Minimum=0, Maximum=2)
      Constrain to Hemisphere=true
      Hemisphere Constraint Vector=&ReturnMeshOrientationAxis(Mesh Orientation Axis=::XAxis, Mesh Orientation Quat=$Particles.SkeletalMeshLocation001.SampledRotation)
     )
    )
    Position Blend:NiagaraFloat=&Lerp_Float(
     A=0
     B=1.1
     Alpha=&Multiply_Float(A=$Particles.PartitionParticles.Normalized Index In Partition, B=$Emitter.NewScale)
    )
    Position End:Vector3f=&Lerp_Vector(
     A=Linked() {linkedVariable {name=Particles.SkeletalMeshLocation001.SampledPosition, type=NiagaraPosition}}
     B=Linked() {linkedVariable {name=Particles.SkeletalMeshLocation002.SampledPosition, type=NiagaraPosition}}
     Alpha=&MakeCustomFloatFromBool(Bool=$Particles.GetLeftOrRight, True Float=1, False Float=0)
    )
    Position Start:NiagaraPosition=$Particles.SkeletalMeshLocation.SampledPosition
    Write Particles Position=true
   )
   &ScaleRibbonWidth(Ribbon Width Scale=&FloatFromCurve(
     FloatCurve=DataInterface() {propertyValues {Curve {keys [
         <at 0 0 RCIM_Cubic>
         {arriveTangentWeight=0.1067, interpMode=RCIM_Cubic, leaveTangentWeight=0.3202, tangentMode=RCTM_User, time=0.2, value=1}
         {arriveTangentWeight=0.3202, interpMode=RCIM_Cubic, leaveTangentWeight=0.1067, tangentMode=RCTM_User, time=0.8, value=1}
         <at 1 0 RCIM_Cubic>
        ]}}}
     Scale Curve=0.6
    ))
   &ScaleColor(
    Scale Alpha=&MakeCustomFloatFromBool(
     Bool=$Particles.Get AlphaMulType
     True Float=&FloatFromCurve(FloatCurve=DataInterface() {propertyValues {Curve {keys [
          {interpMode=RCIM_Cubic, tangentMode=RCTM_User, time=0, value=0}
          {interpMode=RCIM_Cubic, tangentMode=RCTM_User, time=0.5, value=1}
          {tangentMode=RCTM_User, time=1, value=0}
         ]}}})
     False Float=&Waveform("Waveform [1]"=::NewEnumerator0, "[1] Frequency"=8, "[1] Amplitude Min/Max"=<0 1>)
    )
    ScaleA=true
    ScaleRGB=true
   )
   /SetVariables(Particles.RibbonU1Override:NiagaraFloat=&Multiply_Float(
     A=&Add_Float(
      A=&Multiply_Float(A=$Particles.PartitionParticles.Normalized Index In Partition, B=$Emitter.NewScale)
      B=$Particles.GetRandomRibbonOffset
     )
     B=&MakeCustomFloatFromBool(Bool=$Particles.GetLeftOrRight, True Float=1, False Float=-1)
    )) {isSetParameters=true}
   &FN_Light_Attributes(
    Color=&HueShiftLinearColor(Hue=0, Saturation=0.5, Value=1, Linear Color=$Particles.Color)
    Create Color Attribute=true
    Create Exponent Attribute=true
    Create Radius Attribute=true
    Enable Particle Light Enable=true
    Exponent=6
    Light Enabled:NiagaraBool=&Boolean_AndOperation(
     Bool1=&SetBoolByFloatComparison(A=&RandomRangeFloat(Minimum=0, Maximum=1), B=0.8)
     Bool2=&SetBoolByFloatComparison(
      A=&FloatFromCurve(
       FloatCurve=DataInterface() {propertyValues {Curve {keys [<at 0 0>,<at 0.05 0>,<at 0.95 0>,<at 1 1>]}}}
       CurveIndex=&Multiply_Float(A=$Particles.PartitionParticles.Normalized Index In Partition, B=$Emitter.NewScale)
      )
      B=0.5
     )
    )
    Radius=15
    Scale Color=&Multiply_Float(
     A=&Waveform(
      "Waveform [1]"=::NewEnumerator0
      "[1] Frequency"=&RandomRangeFloat(Minimum=1, Maximum=5)
      "[1] Amplitude Min/Max"=<0 1>
      "[1] Bias"=0.5
     )
     B=3
    )
   ) {enabled=false}
   &ScaleSpriteSize(
    Scale Sprite Size Mode=::NewEnumerator0
    Uniform Scale Factor=&Multiply_Float(
     A=&FloatFromCurve(
      FloatCurve=DataInterface() {propertyValues {Curve {keys [
          {arriveTangent=5.6263, interpMode=RCIM_Cubic, leaveTangentWeight=0.9524, time=0, value=0}
          {
           arriveTangent=0.6967
           arriveTangentWeight=0.1114
           interpMode=RCIM_Cubic
           leaveTangent=-5.6928
           leaveTangentWeight=0.9633
           tangentMode=RCTM_Break
           tangentWeightMode=RCTWM_WeightedBoth
           time=0.7128
           value=1
          }
         ]}}}
      Scale Curve=2
     )
     B=&FloatFromCurve(
      FloatCurve=DataInterface() {propertyValues {Curve {keys [<at 0 1>,<at 0.05 0>,<at 0.95 0>,<at 1 0>]}}}
      CurveIndex=&Multiply_Float(A=$Particles.PartitionParticles.Normalized Index In Partition, B=$Emitter.NewScale)
     )
    )
   )
   /SetVariables(Particles.SpriteColor=&MultiplyLinearColor(
     Linear Color=$Particles.Initial.Color
     Scale Factor=&MakeFloat4FromVectorAndFloat(
      "Vector (RGB)"=&VectorFromFloat(Value=2.5)
      "Float (Alpha)"=&FloatFromCurve(FloatCurve=DataInterface() {propertyValues {Curve {keys [<at 0.6089 1>,<at 0.8089 0>]}}})
     )
    )) {isSetParameters=true}
   /SetVariables() {isSetParameters=true}
  ]
  rendererClasses [
   &NiagaraRibbonRendererProperties
   &NiagaraLightRendererProperties
   &NiagaraSpriteRendererProperties
  ]
  renderers [
   {rendererClass=&NiagaraRibbonRendererProperties}
   {rendererClass=&NiagaraLightRendererProperties, rendererIndex=1}
   {rendererClass=&NiagaraSpriteRendererProperties, rendererIndex=2}
  ]
 }
}
summary {
 systemName=NS_Forsake_BackPack
 emitters [
  {
   emitterName=Center_Electricity
   rendererClasses [
    &NiagaraRibbonRendererProperties
    &NiagaraLightRendererProperties
    &NiagaraSpriteRendererProperties
   ]
  }
  {
   emitterName=Side_Electricity
   rendererClasses [
    &NiagaraRibbonRendererProperties
    &NiagaraLightRendererProperties
    &NiagaraSpriteRendererProperties
   ]
  }
  {emitterName=AirGlows, rendererClasses [&NiagaraSpriteRendererProperties, &NiagaraLightRendererProperties]}
  {bEnabled=false, emitterName=AltSparkliesBetter, rendererClasses [&NiagaraSpriteRendererProperties]}
 ]
 userVariables [
  User.IsAlt {type=NiagaraFloat, defaultValue=0, description=""}
  User.SkeletalMesh {
   type=NiagaraDataInterfaceSkeletalMesh
   defaultValue {
    struct=NiagaraExt_VariableValue_DataInterface
    value {dataInterface=&self:NiagaraDataInterfaceSkeletalMesh_0}
   }
  }
 ]
}
